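# Two-sided lit mesh-particle material for a firefly, with animated wing flapping and a glow that changes with time of day.
# Emissive texture, time-of-day tint (defaulting to dark morning and day) and particle color multiply, then pass through HDR remapping, gated by IsLightningBug.
# A sine blends two flap offsets, masked by vertex red and transformed from particle to world space.
# Dynamic parameters drive flap speed, both distances and the glow gate per particle.
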Material /Game/Effects/Fort_Effects/Materials/Fireflies/M_Firefly
graph {
 properties {bUsedWithMeshParticles=true, bUsedWithNanite=true, pixelDepthOffsetMode=PDOM_Legacy, twoSided=true, usedShadingModels=Default Lit}
 nodes [
  n1 = VectorParameter() {parameterName=NoEmissive}
  n2 = VectorParameter() {parameterName=Evning, defaultValue=<rgba 0.154 0.2708 0 0>}
  n3 = VectorParameter() {parameterName=NightTime, defaultValue=<rgba 2 4 0 0>}
  n4 = Multiply(
   A=Multiply(
    A=TextureSample() {texture=/Game/Effects/Fort_Effects/Textures/Bugs/FireFly/T_LightningBug_E}.RGB
    B=MaterialFunctionCall(Color_Morning=@n1.RGB, Color_Day=@n1.RGB, Color_Evening=@n2.RGB, Color_Night=@n3.RGB) {materialFunction=&Color_Time_of_Day}.Result
   )
   B=ParticleColor().RGB
  )
  n5 = Clamp(None=@n4)
  n6 = DynamicParameter() {defaultValue=<rgba 8 1 1 1>, paramNames [FlapSpeed, FlapDistance0, FlapDistance1, IsLightningBug]}
  MaterialFunctionCall(Emissive=LinearInterpolate(
    B=MaterialFunctionCall(LDR Input=@n5, HDR Tint=@n4) {materialFunction=&DeriveHDRfromLDR}.RemappedRange
    Alpha=@n6.IsLightningBug
   )) {materialFunction=&MF_FX_EmissiveAdjustments}
  Multiply(A=TextureSample() {texture=/Game/Effects/Fort_Effects/Textures/Bugs/FireFly/T_LightningBug_D}.RGB) {constB=0.5}
  ScalarParameter() {defaultValue=0.8, parameterName=Roughness}
  n7 = Multiply(A=VectorParameter() {parameterName=Flap0, defaultValue=<rgba 0 0 5 0>}.RGB, B=@n6.FlapDistance0)
  n8 = Multiply(A=VectorParameter() {parameterName=Flap1, defaultValue=<rgba 0 0 -10 0>}.RGB, B=@n6.FlapDistance1)
  n9 = MaterialFunctionCall(Speed=@n6.FlapSpeed) {materialFunction=&DebugTimeSine}
  Transform(None=Multiply(A=LinearInterpolate(A=@n7, B=@n8, Alpha=@n9.Result), B=VertexColor().R)) {transformSourceType=TRANSFORMSOURCE_ParticleWorld}
 ]
}
