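# Two-sided lit material for firefly body mesh particles, usable with Niagara, with a glow that changes with time of day.
# Diffuse texture is tinted by particle color and passed through ConvertFromDiffSpec.
# Emissive texture is multiplied by a constant glow color and the GlowColor dynamic parameter, then fed to MF_FX_EmissiveAdjustments.
# FortniteActiveTimeOfDay from a parameter collection blends that constant from doubled at night to base by day.
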Material /Game/Effects/Fort_Effects/Materials/Fireflies/M_Firefly_Body
graph {
 properties {
  bEnableResponsiveAA=true
  bUsedWithMeshParticles=true
  bUsedWithNiagaraMeshParticles=true
  pixelDepthOffsetMode=PDOM_Legacy
  twoSided=true
  usedShadingModels=Default Lit
 }
 nodes [
  Clamp(None=Sine(None=Time()))
  n1 = Constant3Vector() {constant=<rgba 32.6026 60 0 0>}
  n2 = CollectionParameter() {
   collection=&FortniteMaterialParameters
   parameterName=FortniteActiveTimeOfDay
  }
  n3 = LinearInterpolate(A=Multiply(A=@n1) {constB=2}, B=@n1, Alpha=DotProduct(A=@n2, B=Constant4Vector() {constant=<rgba 1 1 1 0>}.RGBA)) {desc="A=Night B=Day"}
  n4 = DynamicParameter() {defaultValue=<rgba 0 0 0 1>, paramNames [Param1,Param2,Param3,GlowColor]}
  MaterialFunctionCall(Emissive=Multiply(A=TextureSample() {texture=/Game/Effects/Fort_Effects/Textures/Fireflies/T_Firefly_Mesh_01_e}.RGB, B=Multiply(A=@n3, B=@n4.GlowColor))) {materialFunction=&MF_FX_EmissiveAdjustments}
  MaterialFunctionCall(DiffuseColor=Multiply(A=TextureSample() {texture=/Game/Effects/Fort_Effects/Textures/Fireflies/T_Firefly_Mesh_01}.RGB, B=ParticleColor().RGB)) {materialFunction=&ConvertFromDiffSpec}
 ]
}
